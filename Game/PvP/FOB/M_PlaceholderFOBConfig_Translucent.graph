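# Placeholder translucent material for PvP FOB configuration: unlit, additive and two-sided, apparently drawing a tinted border around a fainter interior.
# Multiplied U and V falloffs (each coordinate times its inverse) are reshaped and passed through Ceil, forming a hard interior mask without a texture.
# That mask lerps the Base Color parameter most of the way to black, leaving the edges at full color.
# Fresnel, SphereMask and radial gradient nodes feed nothing visible in the graph.
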Material /Game/PvP/FOB/M_PlaceholderFOBConfig_Translucent
graph {
 properties {
  bScreenSpaceReflections=true
  bTangentSpaceNormal=false
  blendMode=BLEND_Additive
  pixelDepthOffsetMode=PDOM_Legacy
  previewMesh=/Game/Packages/Fortnite_UniqueMaterials_1/StaticMeshes/1x1x1BoxCenterAligned
  shadingModel=MSM_Unlit
  translucencyLightingMode=TLM_Surface
  twoSided=true
  usedShadingModels=Unlit
 }
 nodes [
  Constant() {r=0.6}
  Constant() {r=1}
  Fresnel() {exponent=12}
  n1 = TextureCoordinate()
  n2 = ComponentMask(None=@n1) {r=true}
  n3 = ComponentMask(None=OneMinus(None=@n1)) {r=true}
  n4 = ComponentMask(None=@n1) {g=true}
  n5 = ComponentMask(None=OneMinus(None=@n1)) {g=true}
  n6 = Multiply(A=Multiply(A=Multiply(A=@n2, B=@n3)) {constB=2}, B=Multiply(A=Multiply(A=@n4, B=@n5)) {constB=2})
  n7 = Subtract(A=Multiply(A=Power(Base=@n6) {constExponent=0.5}) {constB=1.1}) {constB=0.06}
  LinearInterpolate(
   A=VectorParameter() {parameterName=Base Color, defaultValue=<rgba 0.0293 0.1186 0.23 1>}.RGB
   B=Constant()
   Alpha=Multiply(A=Ceil(None=@n7)) {constB=0.8}
  )
  MaterialFunctionCall!(UVs, CenterPosition, Radius=Constant() {r=0.57}) {materialFunction=/Engine/Functions/Engine_MaterialFunctions01/Gradient/SphereGradient-2D}
  SphereMask()
 ]
}
